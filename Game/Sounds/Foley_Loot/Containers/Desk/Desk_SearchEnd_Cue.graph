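# Sound cue for the end of searching a desk loot container, played on the loot-collect sound class.
# A quality-level root sends the first two levels to a mix of the shared no-more-loot UI sound and a quieter, randomly modulated desk search-end wave.
# The other two levels play only an optimized desk wave, modulated on the third and plain on the fourth.
# Prop-search attenuation and capped concurrency limit its spread.
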SoundCue /Game/Sounds/Foley_Loot/Containers/Desk/Desk_SearchEnd_Cue
root=QualityLevel_0
nodes [
 {id=Mixer_15, children [WavePlayer_46, Modulator_20], properties {inputVolume [0.8,0.4]}}
 {id=Modulator_0, children [WavePlayer_0], properties {pitchMax=1.07, pitchMin=0.92, volumeMax=1.07, volumeMin=0.92}}
 {id=Modulator_20, children [WavePlayer_78], properties {pitchMax=1.07, pitchMin=0.92, volumeMax=1.07, volumeMin=0.92}}
 {id=QualityLevel_0, children [Mixer_15,Mixer_15,Modulator_0,WavePlayer_1]}
 {id=WavePlayer_0, properties {soundWaveAssetPtr=&/Desk_SearchEnd_opt}}
 {id=WavePlayer_1, properties {soundWaveAssetPtr=&Desk_SearchEnd_opt}}
 {id=WavePlayer_46, properties {soundWaveAssetPtr=/Game/Sounds/Fort_UI_Sounds/Gameplay/Fort_UI_NoMoreLoot}}
 {id=WavePlayer_78, properties {soundWaveAssetPtr=&/Desk_SearchEnd1}}
]
properties {
 attenuationSettings=/Game/Sounds/Attenuation/Fort_Prop_Search_Att
 bOverrideConcurrency=true
 maxDistance=1300
 soundClassObject=&Fort_Loot_Collect
 volumeMultiplier=0.7
 concurrencyOverrides {maxCount=4, resolutionRule=StopFarthestThenPreventNew}
}
