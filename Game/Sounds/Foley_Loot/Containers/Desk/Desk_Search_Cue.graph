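# Desk search foley cue for looting desk containers, routed to the loot-collect sound class with prop-search attenuation.
# A quality-level root picks one of four progressively cheaper branches.
# The richer two mix randomized, quieter desk sounds with louder gear rustles, with modulators adding slight pitch and volume variation.
# The cheaper two play one optimized wave, modulated or plain; a concurrency override blocks new instances at the limit.
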SoundCue /Game/Sounds/Foley_Loot/Containers/Desk/Desk_Search_Cue
root=QualityLevel_0
nodes [
 {id=Mixer_0, children [Modulator_0,Modulator_1], properties {inputVolume [0.4,1.2]}}
 {id=Mixer_21, children [Modulator_21,Random_20], properties {inputVolume [0.4,1.2]}}
 {id=Modulator_0, children [Random_0], properties {pitchMax=1.07, pitchMin=0.92, volumeMax=1.07, volumeMin=0.92}}
 {id=Modulator_1, children [Random_1], properties {pitchMax=1.07, pitchMin=0.92, volumeMax=1.07, volumeMin=0.92}}
 {id=Modulator_2, children [WavePlayer_0], properties {pitchMax=1.07, pitchMin=0.92, volumeMax=1.07, volumeMin=0.92}}
 {id=Modulator_21, children [Random_16], properties {pitchMax=1.07, pitchMin=0.92, volumeMax=1.07, volumeMin=0.92}}
 {id=QualityLevel_0, children [Mixer_21,Mixer_0,Modulator_2,WavePlayer_7]}
 {id=Random_0, children [WavePlayer_1,WavePlayer_2], properties {weights [1,1]}}
 {id=Random_1, children [WavePlayer_4,WavePlayer_5], properties {bRandomizeWithoutReplacement=false, weights [1,1]}}
 {id=Random_16, children [WavePlayer_49, WavePlayer_51, WavePlayer_85], properties {weights [1,1,1]}}
 {id=Random_20, children [WavePlayer_104, WavePlayer_105, WavePlayer_107], properties {weights [1,1,1]}}
 {id=WavePlayer_0, properties {soundWaveAssetPtr=&/Desk_Search_opt}}
 {id=WavePlayer_1, properties {soundWaveAssetPtr=&/Desk_Search11}}
 {id=WavePlayer_104, properties {soundWaveAssetPtr=/Game/Sounds/Foley_Loot/Containers/Misc/Prop_Gear_Search_05}}
 {id=WavePlayer_105, properties {soundWaveAssetPtr=/Game/Sounds/Foley_Loot/Containers/Misc/Prop_Gear_Search_06}}
 {id=WavePlayer_107, properties {soundWaveAssetPtr=/Game/Sounds/Foley_Loot/Containers/Misc/Prop_Gear_Search_01}}
 {id=WavePlayer_2, properties {soundWaveAssetPtr=&/Desk_Search20}}
 {id=WavePlayer_4, properties {soundWaveAssetPtr=&Prop_Gear_Search_05}}
 {id=WavePlayer_49, properties {soundWaveAssetPtr=&Desk_Search11}}
 {id=WavePlayer_5, properties {soundWaveAssetPtr=&Prop_Gear_Search_06}}
 {id=WavePlayer_51, properties {soundWaveAssetPtr=&Desk_Search20}}
 {id=WavePlayer_7, properties {soundWaveAssetPtr=&Desk_Search_opt}}
 {id=WavePlayer_85, properties {soundWaveAssetPtr=&/Desk_Search21}}
]
properties {
 attenuationSettings=/Game/Sounds/Attenuation/Fort_Prop_Search_Att
 bOverrideConcurrency=true
 maxDistance=1300
 soundClassObject=&Fort_Loot_Collect
 concurrencyOverrides {resolutionRule=PreventNew}
}
