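# Sound cue for the foley that plays when searching a junk workbench container completes.
# A single wave player feeds a modulator that slightly randomizes pitch and volume on each play.
# It uses the prop search attenuation and the loot collect sound class at reduced volume.
# A concurrency override caps simultaneous instances, stopping the farthest before preventing new ones.
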SoundCue /Game/Sounds/Foley_Loot/Containers/Desk/Workbench_Junk_SearchComplete_Cue
root=Modulator_14
nodes [
 {id=Modulator_14, children [WavePlayer_58], properties {pitchMax=1.07, pitchMin=0.92, volumeMax=1.07, volumeMin=0.92}}
 {id=WavePlayer_58, properties {soundWaveAssetPtr=&/Workbench_Junk_SearchEnd1}}
]
properties {
 attenuationSettings=/Game/Sounds/Attenuation/Fort_Prop_Search_Att
 bOverrideConcurrency=true
 maxDistance=1300
 soundClassObject=&Fort_Loot_Collect
 volumeMultiplier=0.7
 concurrencyOverrides {maxCount=4, resolutionRule=StopFarthestThenPreventNew}
}
