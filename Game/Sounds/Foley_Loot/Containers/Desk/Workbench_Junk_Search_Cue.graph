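# Sound cue for searching a junk workbench container, using loot-collect sound class and prop-search attenuation.
# A mixer layers a random workbench junk wave, quieter and lightly pitch and volume modulated, with a louder random gear-search wave.
# That shared layer feeds the root mixer twice, directly and through a randomized delay of about one second at lower volume, forming a second rummage.
# A concurrency override prevents new instances at the limit.
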SoundCue /Game/Sounds/Foley_Loot/Containers/Desk/Workbench_Junk_Search_Cue
root=Mixer_34
nodes [
 {id=Delay_54, children [Mixer_23], properties {delayMax=1.2, delayMin=1}}
 {id=Mixer_23, children [Modulator_24,Random_22], properties {inputVolume [0.3,1.2]}}
 {id=Mixer_34, children [Mixer_23,Delay_54], properties {inputVolume [1,0.7]}}
 {id=Modulator_24, children [Random_10], properties {pitchMax=1.07, pitchMin=0.92, volumeMax=1.07, volumeMin=0.92}}
 {id=Random_10, children [WavePlayer_61, WavePlayer_63, WavePlayer_65], properties {weights [1,1,1]}}
 {id=Random_22, children [WavePlayer_114, WavePlayer_115, WavePlayer_117], properties {weights [1,1,1]}}
 {id=WavePlayer_114, properties {soundWaveAssetPtr=/Game/Sounds/Foley_Loot/Containers/Misc/Prop_Gear_Search_05}}
 {id=WavePlayer_115, properties {soundWaveAssetPtr=/Game/Sounds/Foley_Loot/Containers/Misc/Prop_Gear_Search_06}}
 {id=WavePlayer_117, properties {soundWaveAssetPtr=/Game/Sounds/Foley_Loot/Containers/Misc/Prop_Gear_Search_01}}
 {id=WavePlayer_61, properties {soundWaveAssetPtr=&/Workbench_Junk_Search8}}
 {id=WavePlayer_63, properties {soundWaveAssetPtr=&/Workbench_Junk_Search16}}
 {id=WavePlayer_65, properties {soundWaveAssetPtr=&/Workbench_Junk_Search4}}
]
properties {
 attenuationSettings=/Game/Sounds/Attenuation/Fort_Prop_Search_Att
 bOverrideConcurrency=true
 maxDistance=1300
 soundClassObject=&Fort_Loot_Collect
 volumeMultiplier=1.2
 concurrencyOverrides {resolutionRule=PreventNew}
}
